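# Physics asset for the Prisoner backpack cosmetic, giving it collision bodies and joints for a three-link dynamic chain.
# The root bone carries a box and a small sphere, while each chain bone has one short capsule.
# Constraints link the capsules in series from the root, allowing limited swing with twist nearly locked, and the first joint is tighter on one swing axis.
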PhysicsAsset /BRCosmetics/Accessories/FORT_Backpacks/Backpack_M_MED_Prisoner/Meshes/Backpack_M_MED_Prisoner_Physics
bodies {
 backpack_root [
  {shapeType=Sphere, radius=4.9025, center=<-2.0404 -8.4595 1.5607>}
  {shapeType=Box, extentX=25, extentY=7.5, extentZ=25, shapeName=backpack_root_box, center=<-20 -1.7 0>, rotation=<rot 0 5 0>}
 ]
 dny_chain_01 [{shapeType=Capsule, length=3.8648, radius=3.8621, center=<2.006 -0.0002 -0.1081>, rotation=<rot -0.0001 90.0004 -89.9999>}]
 dny_chain_02 [{shapeType=Capsule, length=3.2121, radius=3.9962, center=<1.9452 0.0263 0.0509>, rotation=<rot 90 0 0>}]
 dny_chain_03 [{shapeType=Capsule, length=3.5014, radius=3.7872, center=<2.053 -0.0474 -0.015>, rotation=<rot 90 0 0>}]
}
constraints [
 {bone1Name=dny_chain_01, bone2Name=backpack_root, swing1LimitDegrees=25, swing2LimitDegrees=45, twistLimitDegrees=0.001}
 {bone1Name=dny_chain_02, bone2Name=dny_chain_01, swing1LimitDegrees=45, swing2LimitDegrees=45, twistLimitDegrees=0.001}
 {bone1Name=dny_chain_03, bone2Name=dny_chain_02, swing1LimitDegrees=45, swing2LimitDegrees=45, twistLimitDegrees=0.001}
]
